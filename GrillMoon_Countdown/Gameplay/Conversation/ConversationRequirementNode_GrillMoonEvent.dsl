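; Conversation requirement node for the GrillMoon Countdown that gates a dialogue option on whether the participating player holds event items.
; It gets the player controller from the conversation context and checks the inventory quantity of each entry in EventItems, flagging any owned item.
; ShowWhenHaveItems sets the polarity: the node passes either when an item is found or when none is, and otherwise returns FailedAndHidden.
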(fn IsRequirementSatisfied
  (params (Context :type Conversation%20Context%20Structure%20%28by%20ref%29))
  (LocalVariables|SetFoundItem false)
  (bind _returnvalue (Class|FortConversationContextLibrary|GetParticipantPlayerController Context))
  (LocalVariables|SetPlayerController _returnvalue)
  (for _array_element (Variables|Default|GetEventItems)
    (if (> (Inventory|GetItemQuantityonInventoryOwner (LocalVariables|GetPlayerController) _array_element) 0)
      (LocalVariables|SetFoundItem true)))
  (if (LocalVariables|GetFoundItem)
    (if (Variables|Default|GetShowWhenHaveItems)
      (return "Passed")
      (else
        (return "FailedAndHidden")))
    (elif (Variables|Default|GetShowWhenHaveItems)
      (return "FailedAndHidden")
      (else
        (return "Passed")))))
