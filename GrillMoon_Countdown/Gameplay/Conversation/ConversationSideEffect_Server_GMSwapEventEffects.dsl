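; Server-side conversation side effect for GrillMoon that signals a combat zone event for the player in the conversation, presumably to swap event effects.
; A helper resolves the conversation's player participant, casts it to a Fort player controller and returns its pawn.
; If that pawn is valid, the effect broadcasts a GrillMoonCombatZoneEvent carrying it as a stateful event on the global GameplayEventRouter; otherwise it does nothing.
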(fn GetPlayerReference
  (params (Context :type Conversation%20Context%20Structure%20%28by%20ref%29))
  (bind _returnvalue (Class|FortConversationContextLibrary|GetParticipantActorPlayer Context))
  (Utilities|IsValid _returnvalue
    (:Is%20Valid
      (bind _asfort_player_controller (Utilities|Casting|CastToFortPlayerController _returnvalue)
        (:then
          (return (Game|Player|GetPlayerPawn _asfort_player_controller) true))
        (:CastFailed
          _)))
    (:Is%20Not%20Valid
      _))
  (return 0 false))

(fn ServerCauseSideEffect
  (params (Context :type Conversation%20Context%20Structure%20%28by%20ref%29))
  (bind _playerpawn (Default|GetPlayerReference Context))
  (Utilities|IsValid _playerpawn
    (:Is%20Valid
      (GameplayEventRouter|BroadcastStatefulEvent (GameplayEventRouter|GetGlobalEventRouter) (Utilities|Struct|MakeGrillMoonCombatZoneEvent _playerpawn) 0))
    (:Is%20Not%20Valid
      (return))))
