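# Conversation database for the MistMylar NPC in the GrillMoon countdown, with daily dialogue for Days 1 to 6 and a Day 11 event.
# Each day entry plays one speech, swapping lines by the player's equipped character via HasCID requirements, then hits a shared Continue.
# The Day 11 entry uses GrillMoonEvent requirements to offer Combat Mode or report it active.
# Accepting grants two demon masks and an equipped Masamune, then swaps event effects.
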ConversationDatabase /GrillMoon_Countdown/Gameplay/Conversation/Conversation_MistMylar_GrillMoon
entry_points [
 {tag=Conversation.Entry.MistMylar.GrillMoon.Day11.Event, destinations [#1]}
 {tag=Conversation.Entry.MistMylar.GrillMoon.Day1, destinations [#2]}
 {tag=Conversation.Entry.MistMylar.GrillMoon.Day2, destinations [#3]}
 {tag=Conversation.Entry.MistMylar.GrillMoon.Day3, destinations [#4]}
 {tag=Conversation.Entry.MistMylar.GrillMoon.Day4, destinations [#5]}
 {tag=Conversation.Entry.MistMylar.GrillMoon.Day5, destinations [#6]}
 {tag=Conversation.Entry.MistMylar.GrillMoon.Day6, destinations [#7]}
]
nodes [
 ConversationEntryPointNode_18 {
  id=#7
  class=ConversationEntryPointNode
  outputs [#8]
  properties {entryTag {tagName=Conversation.Entry.MistMylar.GrillMoon.Day6}}
 }
 Speech_C_34 {
  id=#8
  class=Speech_C
  outputs [#9]
  properties {
   nodeName=""
   generalConfig {
    defaultMessage="Soon, I'll be able to breach the Spirit Realm and save my sister... but that's not enough."
    contextualMessages [
     {
      message="Nothing in this world, or any other, will stop me from saving my sister."
      contextRequirements [{requirement {
         class=FortControllerRequirement_HasCID
         properties {allowedCIds [/BRCosmetics/Athena/Items/Cosmetics/Characters/Character_PolishedJade_Mind]}
        }}]
      priority {curve {curveTable=null, rowName=None}}
     }
     {
      message="I have a plan. When things are at their darkest, remember: I know who I am. "
      contextRequirements [{requirement {class=FortControllerRequirement_HasCID, properties {allowedCIds [&Character_CubicVice_Dual]}}}]
      priority {curve {curveTable=null, rowName=None}}
     }
     {
      message="When all of this ends... I wonder whether we will be friends, or enemies."
      contextRequirements [{requirement {
         class=FortControllerRequirement_HasCID
         properties {allowedCIds [
           &Character_VoidRedemption_Rebel
           &Character_BrightDisk
           &Character_IronLilac_Sly
          ]}
        }}]
      priority {curve {curveTable=null, rowName=None}}
     }
     {
      message=I don't know you. I don't trust you. I won't let you touch that Shard.
      contextRequirements [{requirement {
         class=FortControllerRequirement_HasCID
         properties {allowedCIds [&Character_UndergroundRebel_Fashion]}
        }}]
      priority {curve {curveTable=null, rowName=None}}
     }
     {
      message="Whatever happens, watch over my siblings. You owe me that much."
      contextRequirements [{requirement {class=FortControllerRequirement_HasCID, properties {allowedCIds [&Character_RoseDepth_Seed]}}}]
      priority {curve {curveTable=null, rowName=None}}
     }
     {
      message="Try and stop me, demon. I'll enjoy making a fool of you and your master."
      contextRequirements [{requirement {class=FortControllerRequirement_HasCID, properties {allowedCIds [/BRCosmetics/Athena/Items/Cosmetics/Characters/Character_UnluckyRoll]}}}]
      priority {curve {curveTable=null, rowName=None}}
     }
    ]
   }
  }
 }
 Continue_C_1 {id=#9, class=Continue_C}
 ConversationEntryPointNode_4 {
  id=#6
  class=ConversationEntryPointNode
  outputs [#10]
  properties {entryTag {tagName=Conversation.Entry.MistMylar.GrillMoon.Day5}}
 }
 Speech_C_12 {
  id=#10
  class=Speech_C
  outputs [#9]
  properties {
   nodeName=""
   generalConfig {
    defaultMessage="Jade! Jade, I can see you! Look to the sound of my voice!"
    contextualMessages [
     {
      message=I failed my sister once. I will not do so again.
      contextRequirements [{requirement {
         class=FortControllerRequirement_HasCID
         properties {allowedCIds [&Character_PolishedJade_Mind]}
        }}]
      priority {curve {curveTable=null, rowName=None}}
     }
     {
      message="Saving Jade is just the beginning. And, of course, it's on me to save us all."
      contextRequirements [{requirement {class=FortControllerRequirement_HasCID, properties {allowedCIds [&Character_CubicVice_Dual]}}}]
      priority {curve {curveTable=null, rowName=None}}
     }
     {
      message="I'm responsible for my family, and for this Island. You know the feeling, I suspect."
      contextRequirements [{requirement {
         class=FortControllerRequirement_HasCID
         properties {allowedCIds [
           &Character_VoidRedemption_Rebel
           &Character_BrightDisk
           &Character_IronLilac_Sly
          ]}
        }}]
      priority {curve {curveTable=null, rowName=None}}
     }
     {
      message="Are you going to lock the shard away, or use it to your own ends?"
      contextRequirements [{requirement {
         class=FortControllerRequirement_HasCID
         properties {allowedCIds [&Character_UndergroundRebel_Fashion]}
        }}]
      priority {curve {curveTable=null, rowName=None}}
     }
     {
      message="You're formidable. In another life, I think we could have been allies. Or... friends."
      contextRequirements [{requirement {class=FortControllerRequirement_HasCID, properties {allowedCIds [&Character_RoseDepth_Seed]}}}]
      priority {curve {curveTable=null, rowName=None}}
     }
     {
      message=The darkness within the Spirit Realm is holding your leash. I see it now.
      contextRequirements [{requirement {class=FortControllerRequirement_HasCID, properties {allowedCIds [&Character_UnluckyRoll]}}}]
      priority {curve {curveTable=null, rowName=None}}
     }
    ]
   }
  }
 }
 ConversationEntryPointNode_3 {
  id=#5
  class=ConversationEntryPointNode
  outputs [#11]
  properties {entryTag {tagName=Conversation.Entry.MistMylar.GrillMoon.Day4}}
 }
 Speech_C_8 {
  id=#11
  class=Speech_C
  outputs [#9]
  properties {
   nodeName=""
   generalConfig {
    defaultMessage="I can hear its whispers! It knows me, my family, our bloodline. It wants to take me too..."
    contextualMessages [
     {
      message="Soon... soon, Jade will be home. Where she belongs."
      contextRequirements [{requirement {
         class=FortControllerRequirement_HasCID
         properties {allowedCIds [&Character_PolishedJade_Mind]}
        }}]
      priority {curve {curveTable=null, rowName=None}}
     }
     {
      message="The presence in the Spirit Realm, it knows us. Our history. Our blood."
      contextRequirements [{requirement {class=FortControllerRequirement_HasCID, properties {allowedCIds [&Character_CubicVice_Dual]}}}]
      priority {curve {curveTable=null, rowName=None}}
     }
     {
      message="In the end, it always falls on me to fix everything. The genius. The eldest son."
      contextRequirements [{requirement {
         class=FortControllerRequirement_HasCID
         properties {allowedCIds [
           &Character_VoidRedemption_Rebel
           &Character_BrightDisk
           &Character_IronLilac_Sly
          ]}
        }}]
      priority {curve {curveTable=null, rowName=None}}
     }
     {
      message=The Shard's power is seductive. Do you think you're up to resisting it?
      contextRequirements [{requirement {
         class=FortControllerRequirement_HasCID
         properties {allowedCIds [&Character_UndergroundRebel_Fashion]}
        }}]
      priority {curve {curveTable=null, rowName=None}}
     }
     {
      message=I am the only one strong enough to face the darkness in the Spirit Realm.
      contextRequirements [{requirement {class=FortControllerRequirement_HasCID, properties {allowedCIds [&Character_RoseDepth_Seed]}}}]
      priority {curve {curveTable=null, rowName=None}}
     }
     {
      message="You've bound The Night Rose... but you're bound too, aren't you?"
      contextRequirements [{requirement {class=FortControllerRequirement_HasCID, properties {allowedCIds [&Character_UnluckyRoll]}}}]
      priority {curve {curveTable=null, rowName=None}}
     }
    ]
   }
  }
 }
 ConversationEntryPointNode_2 {
  id=#4
  class=ConversationEntryPointNode
  outputs [#12]
  properties {entryTag {tagName=Conversation.Entry.MistMylar.GrillMoon.Day3}}
 }
 Speech_C_7 {
  id=#12
  class=Speech_C
  outputs [#9]
  properties {
   nodeName=""
   generalConfig {
    defaultMessage=The dark presence within the Spirit Realm is powerful. I can feel its malice.
    contextualMessages [
     {
      message="Go away, spirit. You have my sister's face, but not her heart."
      contextRequirements [{requirement {
         class=FortControllerRequirement_HasCID
         properties {allowedCIds [&Character_PolishedJade_Mind]}
        }}]
      priority {curve {curveTable=null, rowName=None}}
     }
     {
      message=There's something dark within the Spirit Realm... It calls to me.
      contextRequirements [{requirement {class=FortControllerRequirement_HasCID, properties {allowedCIds [&Character_CubicVice_Dual]}}}]
      priority {curve {curveTable=null, rowName=None}}
     }
     {
      message=Most people don't realize the pressure that comes with being special. The responsibility.
      contextRequirements [{requirement {
         class=FortControllerRequirement_HasCID
         properties {allowedCIds [
           &Character_VoidRedemption_Rebel
           &Character_BrightDisk
           &Character_IronLilac_Sly
          ]}
        }}]
      priority {curve {curveTable=null, rowName=None}}
     }
     {
      message="Once my sister is rescued, I wonder: what are your plans for the Shard?"
      contextRequirements [{requirement {
         class=FortControllerRequirement_HasCID
         properties {allowedCIds [&Character_UndergroundRebel_Fashion]}
        }}]
      priority {curve {curveTable=null, rowName=None}}
     }
     {
      message="None of them realize, I'm the only one who can handle this power."
      contextRequirements [{requirement {class=FortControllerRequirement_HasCID, properties {allowedCIds [&Character_RoseDepth_Seed]}}}]
      priority {curve {curveTable=null, rowName=None}}
     }
     {
      message="It's pathetic, binding another to your service against their will."
      contextRequirements [{requirement {class=FortControllerRequirement_HasCID, properties {allowedCIds [&Character_UnluckyRoll]}}}]
      priority {curve {curveTable=null, rowName=None}}
     }
    ]
   }
  }
 }
 ConversationEntryPointNode_1 {
  id=#3
  class=ConversationEntryPointNode
  outputs [#13]
  properties {entryTag {tagName=Conversation.Entry.MistMylar.GrillMoon.Day2}}
 }
 Speech_C_6 {
  id=#13
  class=Speech_C
  outputs [#9]
  properties {
   nodeName=""
   generalConfig {
    defaultMessage=I can feel the power of the Spirit Realm. It's beginning.
    contextualMessages [
     {
      message=Try your illusion on some gullible fool. Jade is in the Spirit Realm... for now.
      contextRequirements [{requirement {
         class=FortControllerRequirement_HasCID
         properties {allowedCIds [&Character_PolishedJade_Mind]}
        }}]
      priority {curve {curveTable=null, rowName=None}}
     }
     {
      message="When this works, you'll have to start listening to me."
      contextRequirements [{requirement {class=FortControllerRequirement_HasCID, properties {allowedCIds [&Character_CubicVice_Dual]}}}]
      priority {curve {curveTable=null, rowName=None}}
     }
     {
      message="We're alike, I think. Two people used to holding the world on our shoulders."
      contextRequirements [{requirement {
         class=FortControllerRequirement_HasCID
         properties {allowedCIds [
           &Character_VoidRedemption_Rebel
           &Character_BrightDisk
           &Character_IronLilac_Sly
          ]}
        }}]
      priority {curve {curveTable=null, rowName=None}}
     }
     {
      message="Why did you come here, stranger? What are you looking for?"
      contextRequirements [{requirement {
         class=FortControllerRequirement_HasCID
         properties {allowedCIds [&Character_UndergroundRebel_Fashion]}
        }}]
      priority {curve {curveTable=null, rowName=None}}
     }
     {
      message="I feel a presence in the Spirit Realm... What is it, demon?"
      contextRequirements [{requirement {class=FortControllerRequirement_HasCID, properties {allowedCIds [&Character_RoseDepth_Seed]}}}]
      priority {curve {curveTable=null, rowName=None}}
     }
     {
      message=The Night Rose deserves her freedom. And you deserve... a different fate.
      contextRequirements [{requirement {class=FortControllerRequirement_HasCID, properties {allowedCIds [&Character_UnluckyRoll]}}}]
      priority {curve {curveTable=null, rowName=None}}
     }
    ]
   }
  }
 }
 ConversationEntryPointNode_0 {
  id=#2
  class=ConversationEntryPointNode
  outputs [#14]
  properties {entryTag {tagName=Conversation.Entry.MistMylar.GrillMoon.Day1}}
 }
 Speech_C_5 {
  id=#14
  class=Speech_C
  outputs [#9]
  properties {
   nodeName=""
   generalConfig {
    defaultMessage="Did my brother send you? Tell him I have a plan, and to not get in my way."
    contextualMessages [
     {
      message="Whatever you are, you're not my sister. Stay out of my way."
      contextRequirements [{requirement {
         class=FortControllerRequirement_HasCID
         properties {allowedCIds [&Character_PolishedJade_Mind]}
        }}]
      priority {curve {curveTable=null, rowName=None}}
     }
     {
      message=Brother... don't even think about trying to stop me.
      contextRequirements [{requirement {class=FortControllerRequirement_HasCID, properties {allowedCIds [&Character_CubicVice_Dual]}}}]
      priority {curve {curveTable=null, rowName=None}}
     }
     {
      message=If you're thinking of trying to stop me... don't.
      contextRequirements [{requirement {
         class=FortControllerRequirement_HasCID
         properties {allowedCIds [
           &Character_VoidRedemption_Rebel
           &Character_BrightDisk
           &Character_IronLilac_Sly
          ]}
        }}]
      priority {curve {curveTable=null, rowName=None}}
     }
     {
      message="You stay out of my way, and I'll stay out of yours."
      contextRequirements [{requirement {
         class=FortControllerRequirement_HasCID
         properties {allowedCIds [&Character_UndergroundRebel_Fashion]}
        }}]
      priority {curve {curveTable=null, rowName=None}}
     }
     {
      message=Thanks for your assistance. I hope to return the favor.
      contextRequirements [{requirement {class=FortControllerRequirement_HasCID, properties {allowedCIds [&Character_RoseDepth_Seed]}}}]
      priority {curve {curveTable=null, rowName=None}}
     }
     {
      message="I'm taking my sister back, demon. Don't get in my way."
      contextRequirements [{requirement {class=FortControllerRequirement_HasCID, properties {allowedCIds [&Character_UnluckyRoll]}}}]
      priority {curve {curveTable=null, rowName=None}}
     }
    ]
   }
  }
 }
 ConversationEntryPointNode_17 {
  id=#1
  class=ConversationEntryPointNode
  outputs [#15, #16]
  properties {entryTag {tagName=Conversation.Entry.MistMylar.GrillMoon.Day11.Event}}
 }
 Speech_C_29 {
  id=#16
  class=Speech_C
  outputs [#17]
  properties {nodeName="", generalConfig {defaultMessage=Saving Jade won't be enough. I have to enter the darkness and defeat what lingers there.}}
  sub_nodes [ConversationRequirementNode_GrillMoonEvent_C_0 {class=ConversationRequirementNode_GrillMoonEvent_C}]
 }
 Speech_C_4 {
  id=#17
  class=Speech_C
  outputs [#18, #19]
  properties {nodeName="", generalConfig {defaultMessage=My chaos magic is powered by conflict. Would you like to enter Combat Mode?}}
 }
 Speech_C_3 {
  id=#19
  class=Speech_C
  outputs []
  properties {
   nodeName=""
   softTaskIcon=&T_UI_NPC_Conversation_Give
   generalConfig {defaultMessage="I've gifted you some magical weaponry. Now, FIGHT!"}
  }
  sub_nodes [
   VkConversationSubNode_Choice_2 {class=VkConversationSubNode_Choice, properties {defaultChoiceDisplayText=Yes}}
   ConversationSideEffect_Server_GMGiveItemFromWID_C_0 {class=ConversationSideEffect_Server_GMGiveItemFromWID_C, properties {itemDefinition=/DemonMaskVoid/Gameplay/WID_DemonMask_VoidTeleport}}
   ConversationSideEffect_Server_GMGiveItemFromWID_C_2 {
    class=ConversationSideEffect_Server_GMGiveItemFromWID_C
    properties {equipItem=true, itemDefinition=/MasamuneGameplay/Gameplay/WID_FirePetal_Masamune}
   }
   ConversationSideEffect_Server_GMGiveItemFromWID_C_1 {class=ConversationSideEffect_Server_GMGiveItemFromWID_C, properties {itemDefinition=/DemonMaskFire/Gameplay/WID_DemonMask_Fire}}
   ConversationSideEffect_Server_GMSwapEventEffects_C_0 {class=ConversationSideEffect_Server_GMSwapEventEffects_C}
  ]
 }
 Speech_C_2 {
  id=#18
  class=Speech_C
  outputs []
  properties {nodeName="", generalConfig {defaultMessage=Let me know if you change your mind. You could be part of something great.}}
  sub_nodes [VkConversationSubNode_Choice_0 {class=VkConversationSubNode_Choice, properties {defaultChoiceDisplayText=Decline}}]
 }
 Speech_C_27 {
  id=#15
  class=Speech_C
  outputs [#20]
  properties {nodeName="", generalConfig {defaultMessage="You're already in Combat Mode. Go on and FIGHT!"}}
  sub_nodes [ConversationRequirementNode_GrillMoonEvent_C_3 {class=ConversationRequirementNode_GrillMoonEvent_C, properties {showWhenHaveItems=true}}]
 }
 Continue_C_0 {id=#20, class=Continue_C}
]
